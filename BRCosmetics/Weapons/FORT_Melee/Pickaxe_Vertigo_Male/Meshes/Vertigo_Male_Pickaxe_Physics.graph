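# Physics asset for the male Vertigo pickaxe, a melee weapon cosmetic, defining its simplified collision shape.
# It holds a single body on the lowerarm_r bone, built from two differently oriented capsules, one larger and one slimmer.
# It has no constraints, so there are no jointed parts.
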PhysicsAsset /BRCosmetics/Weapons/FORT_Melee/Pickaxe_Vertigo_Male/Meshes/Vertigo_Male_Pickaxe_Physics
bodies {lowerarm_r [
  {shapeType=Capsule, length=48.1549, radius=14.703, center=<-52.9939 -2.848 -6.3777>, rotation=<rot 6.9987 2.9942 -87.4091>}
  {shapeType=Capsule, length=36.25, radius=7.25, center=<-24.0629 0.0002 -1.5388>, rotation=<rot -89.335 0.0027 -0.0027>}
 ]}
constraints []
